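# UI material that draws the Piew Pistol reticle procedurally from SDF shapes, with no textures.
# Mirrored UVs turn two triangle polygons into corner and lateral markers around a center dot.
# Spread and ADS scalars drive marker offsets, with corner rotation following both and marker scale following ADS.
# Antialiased masks and soft shadow masks feed color and opacity blends, with ADS allowing a white-to-red tint on part of the reticle.
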Material /PiewPistolGameplay/UI/Materials/M_PiewPistol_Reticle
graph {
 properties {bAutomaticallySetUsageInEditor=false, blendMode=BLEND_Translucent, materialDomain=MD_UI, usedShadingModels=Default Lit}
 nodes [
  Constant() {r=0.95}
  Constant() {r=0.91}
  n1 = Constant3Vector() {constant=<rgba 1 1 1 0>}
  n2 = Constant3Vector() {constant=<rgba 1 0 0 0>}
  n3 = NamedRerouteUsage()
  n4 = Ceil(None=@n3)
  n5 = Max(A=@n3, B=NamedRerouteUsage())
  LinearInterpolate(
   A=Constant3Vector() {constant=<rgba 0.2552 0.2552 0.2552 0>}
   B=LinearInterpolate(A=@n1, B=LinearInterpolate(A=@n1, B=@n2, Alpha=@n4), Alpha=ScalarParameter() {parameterName=ADS})
   Alpha=Max(A=NamedRerouteUsage(), B=@n5)
  )
  LinearInterpolate(A=Multiply(A=NamedRerouteUsage()) {constB=0.2}, Alpha=@n5)
  n6 = LinearInterpolate(Alpha=ScalarParameter() {parameterName=Spread}) {constB=0.15}
  n7 = LinearInterpolate(Alpha=ScalarParameter() {parameterName=Spread}) {constA=-0.1, constB=0.35}
  n8 = ScalarParameter() {parameterName=ADS}
  n9 = Add(
   A=MaterialFunctionCall!(UV=MaterialFunctionCall() {materialFunction=&GetUserInterfaceUV}.Normalized UV, Swizzle UV) {materialFunction=/FortUILibrary/MaterialFunctions/Basics/MF_UI_Symmetry_2x2_New}.Symmetry
   B=AppendVector(A=LinearInterpolate(A=@n6, B=@n7, Alpha=@n8) {constB=-0.1}, B=LinearInterpolate(Alpha=@n8) {constA=-0.25, constB=-0.3}) {desc=Corners Offset}
  )
  n10 = ScalarParameter() {parameterName=ADS}
  n11 = LinearInterpolate(
   A=LinearInterpolate(Alpha=ScalarParameter() {parameterName=Spread}) {constA=0.74, constB=0.72}
   B=LinearInterpolate(Alpha=ScalarParameter() {parameterName=Spread}) {constA=0.68, constB=0.72}
   Alpha=@n10
  ) {constA=0.74, constB=0.68, desc=Corners Rotation}
  n12 = Constant2Vector() {g=0.5, r=0.2}
  n13 = LinearInterpolate(Alpha=@n10) {constA=1, constB=0.8, desc=Corners Scale}
  n14 = MaterialFunctionCall!(
   UVs=MaterialFunctionCall(
    UVs=MaterialFunctionCall(UVs=@n9, "Rotation Angle (0-1)"=@n11) {materialFunction=&CustomRotator}.Rotated Values
    Texture Scale=Multiply(A=@n12, B=@n13) {constB=2}
   ) {materialFunction=&ScaleUVsByCenter}.UVs
   Size
   Sides=Constant() {r=3}
   InnerAngle
   OutlineThickness
   OutlineAlignment
   GlowMin
   GlowMax
  ) {materialFunction=/FortUILibrary/MaterialFunctions/Shapes/MF_UI_RegularPolygon}
  n15 = Constant() {r=0.95}
  NamedRerouteDeclaration(None=MaterialFunctionCall(DistanceField=OneMinus(None=@n14.SDF), Threshold=@n15) {materialFunction=&MF_UI_AntiAliasing}.Outer) {name=Corners, nodeColor=<rgba 0 0 0 1>}
  n16 = MaterialFunctionCall(Size=Constant() {r=0.02}) {materialFunction=&MF_UI_Circle}
  NamedRerouteDeclaration(None=OneMinus(None=SmoothStep(Value=@n16.SDF) {constMax=0.015, constMin=0.005})) {name=Circle, nodeColor=<rgba 0 0 0 1>}
  NamedRerouteDeclaration(None=OneMinus(None=SmoothStep(Value=@n16.SDF) {constMax=0.02, constMin=0.007})) {name=CircleShadow, nodeColor=<rgba 0 0 0 1>}
  n17 = LinearInterpolate(Alpha=ScalarParameter() {parameterName=Spread}) {constA=0.18, constB=0.35}
  n18 = LinearInterpolate(Alpha=ScalarParameter() {parameterName=Spread}) {constA=0.1, constB=0.4}
  n19 = ScalarParameter() {parameterName=ADS}
  n20 = Add(
   A=MaterialFunctionCall!(UV=MaterialFunctionCall() {materialFunction=&GetUserInterfaceUV}.Normalized UV, Mirror Y, Swizzle UV) {materialFunction=/FortUILibrary/MaterialFunctions/Basics/MF_UI_Symmetry_2x1_New}.Symmetry
   B=AppendVector(A=LinearInterpolate(A=@n17, B=@n18, Alpha=@n19) {constA=0.18, constAlpha=0, constB=0.1, desc=Lateral Offset}, B=Constant())
  )
  n21 = Constant() {r=0.75}
  n22 = Constant2Vector() {g=0.6, r=0.1}
  n23 = LinearInterpolate(Alpha=ScalarParameter() {parameterName=ADS}) {constA=1.4, constB=1.2, desc=Lateral Scale}
  n24 = MaterialFunctionCall(
   UVs=MaterialFunctionCall(
    UVs=MaterialFunctionCall(UVs=@n20, "Rotation Angle (0-1)"=@n21) {materialFunction=&CustomRotator}.Rotated Values
    Texture Scale=Multiply(A=@n22, B=@n23) {constB=1.4}
   ) {materialFunction=&ScaleUVsByCenter}.UVs
   Sides=Constant() {r=3}
  ) {materialFunction=&MF_UI_RegularPolygon}
  NamedRerouteDeclaration(None=MaterialFunctionCall(DistanceField=OneMinus(None=@n24.SDF), Threshold=@n15) {materialFunction=&MF_UI_AntiAliasing}.Outer) {name=Middle, nodeColor=<rgba 0 0 0 1>}
  n25 = Min(A=@n24.SDF, B=@n14.SDF)
  NamedRerouteDeclaration(None=MaterialFunctionCall(DistanceField=OneMinus(None=@n25), Threshold=Constant() {r=0.93}, Smoothing=Constant() {r=0.1}) {materialFunction=&MF_UI_AntiAliasing}.Outer) {name=Shadows, nodeColor=<rgba 0 0 0 1>}
 ]
}
