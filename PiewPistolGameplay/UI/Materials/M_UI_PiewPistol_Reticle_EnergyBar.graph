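# Translucent UI material that draws the Piew Pistol reticle's energy bar, with a fill set by the Progress parameter.
# Hexagon SDFs clipped by box masks form the bar, bracket and their shadows, and a vertical step cuts the fill.
# Panned lighting texture, soft glow and moving reflected gradient provide animated effect masks.
# The ADS parameter blends in screen-blended color and effects, while Additive suppresses those effects and Opacity scales the final alpha.
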Material /PiewPistolGameplay/UI/Materials/M_UI_PiewPistol_Reticle_EnergyBar
graph {
 properties {bAutomaticallySetUsageInEditor=false, blendMode=BLEND_Translucent, materialDomain=MD_UI, usedShadingModels=Default Lit}
 nodes [
  Constant() {r=0}
  n1 = Constant3Vector() {constant=<rgba 1 1 1 0>}
  n2 = ScalarParameter() {defaultValue=1, parameterName=Additive}
  n3 = NamedRerouteUsage()
  n4 = VectorParameter() {parameterName=Color_Fx, defaultValue=<rgba 1 0.1104 0.235 1>}
  n5 = LinearInterpolate(A=@n4.RGB, B=VectorParameter() {parameterName=Color_Glow, defaultValue=<rgba 1 0.2195 0.2031 1>}.RGB, Alpha=NamedRerouteUsage()) {constB=10}
  n6 = MaterialFunctionCall!(
   Blend=Multiply(A=OneMinus(None=@n2), B=Multiply(A=@n3, B=@n5))
   Base=LinearInterpolate(A=VectorParameter() {desc=ADS Color, parameterName=Color_Bar, defaultValue=<rgba 1 0 0 1>}.RGB, B=@n4.RGB, Alpha=NamedRerouteUsage()) {constA=1}
  ) {materialFunction=/Engine/Functions/Engine_MaterialFunctions03/Blends/Blend_Screen}
  n7 = ScalarParameter() {parameterName=ADS}
  n8 = NamedRerouteUsage()
  n9 = Max(A=NamedRerouteUsage(), B=LinearInterpolate(B=NamedRerouteUsage(), Alpha=ScalarParameter() {parameterName=ADS}))
  LinearInterpolate(
   A=LinearInterpolate(A=Constant3Vector(), B=LinearInterpolate(A=@n1, B=@n6.Result, Alpha=@n7), Alpha=Multiply(A=@n8, B=@n9))
   B=@n1
   Alpha=NamedRerouteUsage()
  ) {constAlpha=0}
  n10 = NamedRerouteUsage()
  n11 = ScalarParameter() {defaultValue=1, parameterName=Additive}
  n12 = NamedRerouteUsage()
  n13 = Multiply(A=@n11, B=Multiply(A=NamedRerouteUsage(), B=@n12))
  n14 = NamedRerouteUsage()
  n15 = Constant() {desc=GlowOpacity, r=0.2}
  n16 = NamedRerouteUsage()
  n17 = Add(
   A=Multiply(A=Multiply(A=@n12, B=@n14), B=NamedRerouteUsage())
   B=Multiply(A=NamedRerouteUsage(), B=LinearInterpolate(A=@n15, Alpha=@n16) {constB=0.8}) {constB=0.5}
  )
  n18 = OneMinus(None=@n11)
  n19 = LinearInterpolate(A=@n13, B=Max(A=@n13, B=Multiply(A=@n17, B=@n18)), Alpha=ScalarParameter() {parameterName=ADS})
  n20 = OneMinus(None=Multiply(A=Constant() {desc=BackgroundOpacity, r=0.5}, B=ScalarParameter() {defaultValue=1, parameterName=Additive}))
  n21 = Multiply(A=NamedRerouteUsage(), B=ScalarParameter() {defaultValue=1, parameterName=Additive})
  Multiply(
   A=Max(A=LinearInterpolate(A=@n10, B=@n19, Alpha=@n20), B=Multiply(A=@n21) {constB=0.4})
   B=ScalarParameter() {defaultValue=0.9, parameterName=Opacity}
  )
  n22 = MaterialFunctionCall!(
   UVs=Add(A=TextureCoordinate(), B=Constant2Vector() {r=0.53})
   Size=Constant() {r=1.3}
   Sides=Constant() {r=6}
   InnerAngle
   OutlineThickness=Constant() {r=0.001}
   OutlineAlignment
   GlowMin
   GlowMax
  ) {materialFunction=/FortUILibrary/MaterialFunctions/Shapes/MF_UI_RegularPolygon}
  n23 = MaterialFunctionCall!(DistanceFieldA=@n22.SDF, DistanceFieldB=Subtract(A=@n22.SDFOutline) {constB=0.05}, BlendAmount) {materialFunction=/FortUILibrary/MaterialFunctions/Shapes/Modifiers/MF_UI_SDF_Subtraction}
  n24 = Multiply(
   A=MaterialFunctionCall(Size=Constant2Vector() {g=0.85, r=0.14}) {materialFunction=&MF_UI_Box}.Fill
   B=MaterialFunctionCall(DistanceField=@n23.Smooth) {materialFunction=&MF_UI_AntiAliasing}.Inner
  )
  NamedRerouteDeclaration(None=@n24) {name=Bar, nodeColor=<rgba 0 0 0 1>}
  n25 = OneMinus(None=@n22.SDFOutline)
  n26 = MaterialFunctionCall(Size=Constant2Vector() {g=0.85, r=0.18}) {materialFunction=&MF_UI_Box}
  n27 = Multiply(
   A=MaterialFunctionCall(DistanceField=@n25, Threshold=Constant() {r=0.985}) {materialFunction=&MF_UI_AntiAliasing}.Outer
   B=@n26.Fill
  )
  NamedRerouteDeclaration(None=@n27) {name=Bracket, nodeColor=<rgba 0 0 0 1>}
  n28 = Multiply(A=TextureCoordinate()) {constB=1.5}
  n29 = Time()
  n30 = Constant()
  n31 = LinearInterpolate(
   A=Frac(None=Multiply(A=@n29) {constB=2})
   B=Multiply(A=@n29) {constB=2}
   Alpha=MaterialFunctionCall(Speed=Multiply(A=@n30) {constB=0.5}) {materialFunction=&DebugTimeSine}.Result
  ) {constAlpha=1, constB=0}
  n32 = Power(
   Base=TextureSample(UVs=Panner(Coordinate=@n28, Time=@n31) {speedX=0.1, speedY=0.25}) {samplerType=SAMPLERTYPE_LinearColor, texture=/PiewPistolGameplay/UI/Textures/T_UI_PiewPistol_Lighting}.RGB
   Exp=LinearInterpolate(Alpha=NamedRerouteUsage()) {constA=5, constB=2}
  ) {constExponent=5}
  n33 = Multiply(A=@n32) {constB=0.5}
  n34 = Multiply(A=@n32) {constB=3}
  n35 = NamedRerouteUsage()
  NamedRerouteDeclaration(None=Multiply(A=LinearInterpolate(A=@n33, B=@n34, Alpha=@n35) {constB=3}) {constB=0.5}) {name=Lighting Fx, nodeColor=<rgba 0 0 0 1>}
  n36 = ComponentMask(None=TextureCoordinate()) {g=true}
  n37 = Clamp(None=ScalarParameter() {defaultValue=0.7, parameterName=Progress})
  NamedRerouteDeclaration(None=Step(Y=OneMinus(None=@n36), X=LinearInterpolate(Alpha=@n37) {constA=0.163, constB=0.837})) {name=FillProgress, nodeColor=<rgba 0 0 0 1>}
  n38 = TextureCoordinate()
  n39 = MaterialFunctionCall(UVs=Add(A=@n38, B=Constant2Vector() {g=0.16, r=0.268}), Size=Constant2Vector() {g=1, r=0.4}) {materialFunction=&MF_UI_Box}
  n40 = NamedRerouteUsage()
  n41 = Add(A=@n38, B=Constant2Vector() {r=-0.25})
  n42 = Constant2Vector() {g=1, r=0.7}
  n43 = Subtract(A=OneMinus(None=MaterialFunctionCall(UVs=@n41, Size=@n42) {materialFunction=&MF_UI_Box}.SDF)) {constB=0.8}
  NamedRerouteDeclaration(None=Multiply(A=Multiply(A=@n39.Fill, B=@n40), B=Multiply(A=@n43) {constB=1.5})) {name=Glow, nodeColor=<rgba 0 0 0 1>}
  n44 = MaterialFunctionCall(DistanceField=@n25, Threshold=Constant() {r=0.97}) {materialFunction=&MF_UI_AntiAliasing}
  NamedRerouteDeclaration(None=LinearInterpolate(A=@n27, B=Multiply(A=@n44.Outer, B=@n26.Fill), Alpha=ScalarParameter() {defaultValue=0.3, parameterName=ShadowOpacity}) {constAlpha=0.3}) {name=BracketShadow, nodeColor=<rgba 0 0 0 1>}
  n45 = MaterialFunctionCall(DistanceField=@n23.Smooth, Threshold=Constant() {desc=ShadowThickness, r=0.022}) {materialFunction=&MF_UI_AntiAliasing}
  n46 = MaterialFunctionCall(Size=Constant2Vector() {g=0.85, r=0.165}) {materialFunction=&MF_UI_Box}
  NamedRerouteDeclaration(None=LinearInterpolate(A=@n24, B=Multiply(A=@n45.Inner, B=@n46.Fill), Alpha=ScalarParameter() {defaultValue=0.3, parameterName=ShadowOpacity}) {constAlpha=0.4}) {name=BarShadow, nodeColor=<rgba 0 0 0 1>}
  NamedRerouteDeclaration(None=MaterialFunctionCall(Speed=Constant() {r=1}) {materialFunction=&DebugTimeSine}.Result) {name=AnimTime, nodeColor=<rgba 0 0 0 1>}
  n47 = Time()
  n48 = Constant() {r=0.3}
  n49 = MaterialFunctionCall!(UVs=TextureCoordinate(), Offset=Frac(None=Multiply(A=@n47, B=@n48)), RotationCenter=Constant() {r=0.8}, RotationAngle=Constant() {r=0.25}) {materialFunction=/FortUILibrary/MaterialFunctions/Gradients/MF_UI_ReflectedGradient}
  NamedRerouteDeclaration(None=SmoothStep(Value=Power(Base=@n49.Result)) {constMax=0.8, constMin=0.5}) {name=Gradient, nodeColor=<rgba 0 0 0 1>}
 ]
}
