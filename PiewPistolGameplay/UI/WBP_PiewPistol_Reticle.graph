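# Reticle widget for the Piew pistol, showing aim state, firing spread and a weapon overheat meter.
# View target changes rebind it to the equipped pistol's targeting, ammo and server overheat events.
# Animated material parameters drive ADS, reversing mid-play from the current time, and a spread pulse per shot.
# Overheat sets inverted meter progress and opacity on base and additive bars, with a timer for fading them out.
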WidgetBlueprint /PiewPistolGameplay/UI/WBP_PiewPistol_Reticle
animations [
 Anim_ADS {
  binding_count=3
  length_seconds=0.1
  object=&self:Anim_ADS
  sequence {bindings [
    Reticle {
     id=#1
     possessed_class=Image
     tracks [{
       type=MovieSceneWidgetMaterialTrack
       sections [{type=MovieSceneParameterSection, end_frame=2, start_frame=0, channels [ADS {type=MovieSceneScriptingFloatChannel, keys [[0,0],[2,1]]}]}]
      }]
    }
    OverheatBar {
     id=#2
     possessed_class=Image
     tracks [{
       type=MovieSceneWidgetMaterialTrack
       sections [{type=MovieSceneParameterSection, end_frame=2, start_frame=0, channels [ADS {type=MovieSceneScriptingFloatChannel, keys [[0,0],[2,1]]}]}]
      }]
    }
    OverheatBar_Additive {
     id=#3
     possessed_class=Image
     tracks [{
       type=MovieSceneWidgetMaterialTrack
       sections [{type=MovieSceneParameterSection, end_frame=2, start_frame=0, channels [ADS {type=MovieSceneScriptingFloatChannel, keys [[0,0],[2,1]]}]}]
      }]
    }
   ]}
 }
 Anim_Spread {
  binding_count=1
  length_seconds=0.1
  object=&self:Anim_Spread
  sequence {bindings [Reticle {
     id=#4
     possessed_class=Image
     tracks [{
       type=MovieSceneWidgetMaterialTrack
       sections [{type=MovieSceneParameterSection, end_frame=2, start_frame=0, channels [Spread {type=MovieSceneScriptingFloatChannel, keys [[0,0],[2,1]]}]}]
      }]
    }]}
 }
 Anim_Overheat_FadeOut {
  binding_count=2
  length_seconds=0.25
  object=&self:Anim_Overheat_FadeOut
  sequence {bindings [
    OverheatBar_Additive {
     id=#5
     possessed_class=Image
     tracks [{
       type=MovieSceneWidgetMaterialTrack
       sections [{type=MovieSceneParameterSection, end_frame=5, start_frame=0, channels [Opacity {type=MovieSceneScriptingFloatChannel, keys [[0,1],[5,0]]}]}]
      }]
    }
    OverheatBar {
     id=#6
     possessed_class=Image
     tracks [{
       type=MovieSceneWidgetMaterialTrack
       sections [{type=MovieSceneParameterSection, end_frame=5, start_frame=0, channels [Opacity {type=MovieSceneScriptingFloatChannel, keys [[0,1],[5,0]]}]}]
      }]
    }
   ]}
 }
 Anim_FadeIn {
  binding_count=1
  length_seconds=0.25
  object=&self:Anim_FadeIn
  sequence {bindings [Overlay_14 {
     id=#7
     possessed_class=Overlay
     tracks [{
       type=MovieSceneFloatTrack
       display_name=Render Opacity
       sections [{type=MovieSceneFloatSection, end_frame=5, start_frame=0, channels [None {type=MovieSceneScriptingFloatChannel, keys [[0,0],[5,1]]}]}]
      }]
    }]}
 }
]
graphs [
 BindEventFunctions {nodes [
   BindEventFunctions!() -> (then, Bind:bool=~) {is=title K2Node_FunctionEntry}
   > Is Valid(InputObject=Get Weapon() -> (Weapon:BP Pistol Piew Athena Object Reference))
   > n1 = Branch(Condition=Get Bind() -> (Bind:bool))
   n2 = Create Event()
   > Bind Event to On Targeting Changed(self:Fort Weapon Object Reference=Get Weapon(), Delegate=@n2) {is=title K2Node_AddDelegate}
   n3 = Create Event()
   > Bind Event to On Consume Ammo(self:Fort Weapon Object Reference=Get Weapon(), Delegate=@n3) {is=title K2Node_AddDelegate}
   n4 = Create Event()
   > Bind Event to On Server Overheat Changed(self:BP Pistol Piew Athena Object Reference=Get Weapon(), Delegate=@n4) {is=title K2Node_AddDelegate}
   Unbind Event from On Targeting Changed(execute=@n1.else, self:Fort Weapon Object Reference=Get Weapon(), Delegate=@n2) {is=title K2Node_RemoveDelegate}
   > Unbind Event from On Consume Ammo(self:Fort Weapon Object Reference=Get Weapon(), Delegate=@n3) {is=title K2Node_RemoveDelegate}
   > Unbind Event from On Server Overheat Changed(self:BP Pistol Piew Athena Object Reference=Get Weapon(), Delegate=@n4) {is=title K2Node_RemoveDelegate}
  ]}
 EventGraph {nodes [
   n1 = Handle_TargetingChanged!() -> (OutputDelegate, then, bIsTargeting:bool=~) {is=title K2Node_CustomEvent}
   > n2 = Delta Gate Wildcard!(execute:Exec, Variable:"Boolean (by ref)"=Get IsInADS() -> (IsInADS:bool), New Value:bool=@n1.bIsTargeting) -> (Success:Exec, Updated Variable:"Boolean (by ref)", Previous Variable:bool=~) {is=title K2Node_MacroInstance}
   > n3 = Branch(Condition=IsAnimationPlaying(InAnimation=Get Anim_ADS() -> (Anim_ADS:Widget Animation Object Reference)))
   > StopAnimation(InAnimation=Get Anim_ADS())
   > PlayAnimationTimeRange!(
    execute
    self:User Widget Object Reference
    InAnimation:Widget Animation Object Reference=Get Anim_ADS()
    StartAtTime:float=GetAnimationCurrentTime!(self:User Widget Object Reference, InAnimation:Widget Animation Object Reference=Get Anim_ADS()) -> (ReturnValue:float)
    EndAtTime:float=SelectFloat(A=1, B=0, bPickA=@n2)
    NumLoopsToPlay:int=1
    PlayMode:EUMGSequencePlayMode Enum
    PlaybackSpeed:float=1
    bRestoreState:bool=false
   ) -> (then, ReturnValue:Widget Animation Handle Structure)
   n4 = Branch(execute=@n3.else, Condition=@n2)
   > PlayAnimationForward(InAnimation=Get Anim_ADS())
   PlayAnimationReverse(execute=@n4.else, InAnimation=Get Anim_ADS())
   OnViewTargetChange!() -> (
    OutputDelegate
    then
    PlayerController:Fort Player Controller Object Reference
    OldViewTarget:Actor Object Reference
    NewViewTarget:Actor Object Reference
   ) {is=title K2Node_CustomEvent}
   > n5 = Sequence()
   > BindEventFunctions!(execute, self, Bind:bool=false) -> (then) {is=title K2Node_CallFunction}
   > n6 = Cast To BP_Pistol_Piew_Athena!(
    execute
    Object=GetEquippedWeapon!(self:Fort Weapon Reticle Extension Widget Base Object Reference) -> (ReturnValue:Fort Weapon Object Reference)
   ) -> (then, CastFailed, AsBP Pistol Piew Athena:BP Pistol Piew Athena Object Reference) {is=title K2Node_DynamicCast}
   > Set Weapon(Weapon:BP Pistol Piew Athena Object Reference=@n6) -> (Output_Get:BP Pistol Piew Athena Object Reference)
   > BindEventFunctions(Bind=true)
   Set IsOverheated(execute=@n5.then_1, IsOverheated:bool=IsOverheated!(self:Fort Weapon Ranged Object Reference=Get Weapon()) -> (ReturnValue:bool)) -> (Output_Get:bool)
   > Server_OnOverheatChanged()
   > Set IsInADS(IsInADS:bool=false) -> (Output_Get:bool)
   > Handle_TargetingChanged!(execute, self, bIsTargeting:bool=IsTargeting!(self:Fort Weapon Object Reference=Get Weapon()) -> (ReturnValue:bool)) -> (then) {is=title K2Node_CallFunction}
   Server_OnOverheatChanged() {is=title K2Node_CustomEvent}
   n7 = Get Weapon()
   n8 = Get ServerOverheat!(self:BP Pistol Piew Athena Object Reference=@n7) -> (ServerOverheat:double)
   > Set CurrentOverheatValue(CurrentOverheatValue:double=@n8) -> (Output_Get:double)
   n9 = float - float(A=1, B=float / float(A=@n8, B=GetOverheatingMaxValue!(self:Fort Weapon Ranged Object Reference=@n7) -> (ReturnValue:float)))
   > SetScalarParameterValue(self=Get ChargeMeter_MAT() -> (ChargeMeter_MAT:Material Instance Dynamic Object Reference), ParameterName=Progress, Value=@n9)
   > SetScalarParameterValue(
    self=Get OverheatBarAdditiveMeter_MAT() -> (OverheatBarAdditiveMeter_MAT:Material Instance Dynamic Object Reference)
    ParameterName=Progress
    Value=@n9
   )
   > n10 = Branch(Condition="Nearly Equal (Float)"(A=@n9, B=0, ErrorTolerance=0.04))
   > n11 = Set IsOverheated(IsOverheated=true)
   n12 = SelectFloat(A=0.6, B=1, bPickA=Get IsOverheated() -> (IsOverheated:bool))
   SetScalarParameterValue(execute=[@n11.then, @n16.then, @n15.else], self=Get ChargeMeter_MAT(), ParameterName=Opacity, Value=@n12)
   > SetScalarParameterValue(self=Get OverheatBarAdditiveMeter_MAT(), ParameterName=Opacity, Value=@n12)
   > n13 = Set Timer by Event(Delegate=Create Event(), Time=Get OverheatBarHideTime() -> (OverheatBarHideTime:double), bLooping=false)
   > Set HideTimerHandle(HideTimerHandle:Timer Handle Structure=@n13) -> (Output_Get:Timer Handle Structure)
   n14 = Get CurrentOverheatValue() -> (CurrentOverheatValue:double)
   n15 = Branch(execute=@n10.else, Condition=AND Boolean(A=Get IsOverheated(), B="Equal (Float)"(A=@n14)))
   > n16 = Set IsOverheated(IsOverheated=false)
   HideOverheatMeter() {is=title K2Node_CustomEvent}
   > PlayAnimation(InAnimation=Get Anim_Overheat_FadeOut() -> (Anim_Overheat_FadeOut:Widget Animation Object Reference))
   > Clear and Invalidate Timer by Handle(Handle=Get HideTimerHandle() -> (HideTimerHandle:Timer Handle Structure))
   On Consume Ammo!() -> (OutputDelegate, then, Quantity:int) {is=title K2Node_CustomEvent}
   > PlayAnimation(
    InAnimation=Get Anim_Spread() -> (Anim_Spread:Widget Animation Object Reference)
    PlayMode=Reverse
    PlaybackSpeed=SelectFloat(A=0.65, B=1, bPickA=Get IsInADS())
   )
   Event OnInitialized()
   > n17 = GetDynamicMaterial(self=Get OverheatBar() -> (OverheatBar:Image Object Reference))
   > Set ChargeMeter_MAT(ChargeMeter_MAT:Material Instance Dynamic Object Reference=@n17) -> (Output_Get:Material Instance Dynamic Object Reference)
   > n18 = GetDynamicMaterial(self=Get Reticle() -> (Reticle:Image Object Reference))
   > Set Ret_MAT(Ret_MAT:Material Instance Dynamic Object Reference=@n18) -> (Output_Get:Material Instance Dynamic Object Reference)
   > n19 = GetDynamicMaterial(self=Get OverheatBar_Additive() -> (OverheatBar_Additive:Image Object Reference))
   > Set OverheatBarAdditiveMeter_MAT(OverheatBarAdditiveMeter_MAT:Material Instance Dynamic Object Reference=@n19) -> (Output_Get:Material Instance Dynamic Object Reference)
   Event Construct()
   n20 = GetOwningPlayer()
   > n21 = Cast To FortPlayerController(Object=@n20)
   > n22 = Set owningFortPlayerController(owningFortPlayerController:Fort Player Controller Object Reference=@n21) -> (Output_Get:Fort Player Controller Object Reference)
   > Bind Event to On View Target Changed(self:Fort Player Controller Object Reference=@n22, Delegate=Create Event()) {is=title K2Node_AddDelegate}
   > PlayAnimation(InAnimation=Get Anim_FadeIn() -> (Anim_FadeIn:Widget Animation Object Reference))
   > OnViewTargetChange!(
    execute
    self
    PlayerController:Fort Player Controller Object Reference
    OldViewTarget:Actor Object Reference
    NewViewTarget:Actor Object Reference=GetViewTarget!(self:Controller Object Reference=@n20) -> (ReturnValue:Actor Object Reference)
   ) -> (then) {is=title K2Node_CallFunction}
   Event Destruct()
   n23 = Get owningFortPlayerController() -> (owningFortPlayerController:Fort Player Controller Object Reference)
   > Is Valid(InputObject=@n23)
   > Unbind Event from On View Target Changed(self:Fort Player Controller Object Reference=@n23, Delegate=Create Event()) {is=title K2Node_RemoveDelegate}
   > Is Valid(InputObject=Get Weapon())
   > BindEventFunctions(Bind=false)
   > Set owningFortPlayerController()
   > Set Weapon()
   > SetScalarParameterValue(self=Get Ret_MAT() -> (Ret_MAT:Material Instance Dynamic Object Reference), ParameterName=ADS, Value=0)
   > Clear and Invalidate Timer by Handle(Handle=Get HideTimerHandle())
  ]}
 ValidateWeapon {nodes [
   ValidateWeapon() {is=title K2Node_FunctionEntry}
   > n1 = Is Valid(InputObject=Get Weapon())
   n2 = Cast To BP_Pistol_Piew_Athena(execute=@n1.Is Not Valid, Object=GetEquippedWeapon())
   > n3 = Set Weapon(Weapon=@n2)
   Return Node!(execute=[@n1.Is Valid, @n3.then], Weapon:BP Pistol Piew Athena Object Reference=Get Weapon())
  ]}
]
variables [
 ChargeMeter_MAT {type=MaterialInstanceDynamic}
 CurrentOverheatValue {type=number}
 HideTimerHandle {type=TimerHandle}
 IsInADS {type=boolean}
 IsOverheated {type=boolean}
 OverheatBarAdditiveMeter_MAT {type=MaterialInstanceDynamic}
 OverheatBarHideTime {type=number, default=5}
 owningFortPlayerController {type=FortPlayerController, default=null}
 Ret_MAT {type=MaterialInstanceDynamic}
 Weapon {type=BP_Pistol_Piew_Athena_C, default=null}
]
widgets {
 info {
  inheritedWidgetCount=1
  parentClass=/Script/FortniteUI.FortWeaponReticleExtensionWidgetBase
  rootWidgetClass=&Overlay
  widgetCount=6
 }
 widgets [
  Overlay_14 {widgetClassPath=&Overlay, widget=Overlay_14}
  Reticle {
   widgetClassPath=&Image
   bIsVariable=true
   parent=Overlay_14
   widget=Reticle
   properties {brush {
     bIsDynamicallyLoaded=false
     drawAs=Image
     imageType=FullColor
     mirroring=NoMirror
     resourceName=None
     resourceObject=/PiewPistolGameplay/UI/Materials/M_PiewPistol_Reticle_Inst
     tiling=NoTile
     imageSize=<80 80>
     margin=<ltrb 0 0 0 0>
     outlineSettings {
      bUseBrushTransparency=false
      roundingType=HalfHeightRadius
      width=0
      color {colorUseRule=UseColor_Specified, specifiedColor=<rgba 0 0 0 0>}
      cornerRadii=<0 0 0 0>
     }
     tintColor {colorUseRule=UseColor_Specified, specifiedColor=<rgba 1 1 1 1>}
     uVRegion {bIsValid=false, max=<0 0>, min=<0 0>}
    }}
   slot {class=OverlaySlot, properties {horizontalAlignment=HAlign_Center, verticalAlignment=VAlign_Center}}
  }
  CenterDot {
   widgetClassPath=&Image
   bIsVariable=true
   parent=Overlay_14
   widget=CenterDot
   properties {brush {
     bIsDynamicallyLoaded=false
     drawAs=Image
     imageType=FullColor
     mirroring=NoMirror
     resourceName=None
     resourceObject=/FortUILibrary/Systems/Reticles/MaterialInstances/MI_Reticle_Center_32px_Dot
     tiling=NoTile
     imageSize=<18 18>
     margin=<ltrb 0 0 0 0>
     outlineSettings {
      bUseBrushTransparency=false
      roundingType=HalfHeightRadius
      width=0
      color {colorUseRule=UseColor_Specified, specifiedColor=<rgba 0 0 0 0>}
      cornerRadii=<0 0 0 0>
     }
     tintColor {colorUseRule=UseColor_Specified, specifiedColor=<rgba 1 1 1 1>}
     uVRegion {bIsValid=false, max=<0 0>, min=<0 0>}
    }}
   slot {class=OverlaySlot, properties {horizontalAlignment=HAlign_Center, verticalAlignment=VAlign_Center}}
  }
  OverheatBar {
   widgetClassPath=&Image
   bIsVariable=true
   parent=Overlay_14
   widget=OverheatBar
   properties {brush {
     bIsDynamicallyLoaded=false
     drawAs=Image
     imageType=FullColor
     mirroring=NoMirror
     resourceName=None
     resourceObject=/PiewPistolGameplay/UI/Materials/MI_UI_PiewPistol_Reticle_Base
     tiling=NoTile
     imageSize=<100 100>
     margin=<ltrb 0 0 0 0>
     outlineSettings {
      bUseBrushTransparency=false
      roundingType=HalfHeightRadius
      width=0
      color {colorUseRule=UseColor_Specified, specifiedColor=<rgba 0 0 0 0>}
      cornerRadii=<0 0 0 0>
     }
     tintColor {colorUseRule=UseColor_Specified, specifiedColor=<rgba 1 1 1 1>}
     uVRegion {bIsValid=false, max=<0 0>, min=<0 0>}
    }}
   slot {class=OverlaySlot, properties {horizontalAlignment=HAlign_Right, verticalAlignment=VAlign_Center, padding=<ltrb 0 0 -120 0>}}
  }
  OverheatBar_Additive {
   widgetClassPath=&Image
   bIsVariable=true
   parent=Overlay_14
   widget=OverheatBar_Additive
   properties {brush {
     bIsDynamicallyLoaded=false
     drawAs=Image
     imageType=FullColor
     mirroring=NoMirror
     resourceName=None
     resourceObject=/PiewPistolGameplay/UI/Materials/MI_UI_PiewPistol_Reticle_Additive
     tiling=NoTile
     imageSize=<100 100>
     margin=<ltrb 0 0 0 0>
     outlineSettings {
      bUseBrushTransparency=false
      roundingType=HalfHeightRadius
      width=0
      color {colorUseRule=UseColor_Specified, specifiedColor=<rgba 0 0 0 0>}
      cornerRadii=<0 0 0 0>
     }
     tintColor {colorUseRule=UseColor_Specified, specifiedColor=<rgba 1 1 1 1>}
     uVRegion {bIsValid=false, max=<0 0>, min=<0 0>}
    }}
   slot {class=OverlaySlot, properties {horizontalAlignment=HAlign_Right, verticalAlignment=VAlign_Center, padding=<ltrb 0 0 -120 0>}}
  }
  TutorialHighlight {widgetClassPath=&TutorialHighlight}
 ]
}
